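# Sound cue for the bounce of the Easter grenade launcher projectile in Figment Season 8.
# A random node picks one of three equally weighted bounce wave variations, then a modulator at the root processes it before output.
# It uses the shared grenade bounce attenuation and plays under the ranged weapon projectile effects sound class.
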SoundCue /Figment_Launcher_Grenade_Athena/Sounds/S08/SC_Figment_S08_Launcher_Easter_Bounce
root=Modulator_0
nodes [
 {id=Modulator_0, children [Random_0]}
 {id=Random_0, children [WavePlayer_3,WavePlayer_4,WavePlayer_5], properties {weights [1,1,1]}}
 {id=WavePlayer_3, properties {soundWaveAssetPtr=&/Launcher_Easter_BounceA_03}}
 {id=WavePlayer_4, properties {soundWaveAssetPtr=&/Launcher_Easter_BounceA_02}}
 {id=WavePlayer_5, properties {soundWaveAssetPtr=&/Launcher_Easter_BounceA_01}}
]
properties {
 attenuationSettings=/Game/Sounds/Attenuation/Fort_Grenade_Bounce_Att
 maxDistance=5500
 soundClassObject=&Fort_WeaponRanged_Projectile_Effects
 volumeMultiplier=1
}
