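# Sound cue for the explosion of the Figment S08 Easter grenade launcher, mixing three layers.
# The main layer randomly picks one of three explosion waves, with medium-explosion attenuation and sound class.
# The Distant_01 wave also uses the medium settings at reduced mixer volume, while the Medium_01 wave uses the distant ones, opposite to their names.
# Each layer passes through a Modulator and overrides the cue's sound class.
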SoundCue /Figment_Launcher_Grenade_Athena/Sounds/S08/SC_Figment_S08_Launcher_Easter_Explode
root=Mixer_0
nodes [
 {
  id=Attenuation_0
  children [Modulator_0]
  properties {attenuationSettings=/Game/Sounds/Attenuation/Fort_Weapon_Rocket_Medium_Explosion_Athena_Att}
 }
 {
  id=Attenuation_3
  children [Modulator_2]
  properties {attenuationSettings=/Game/Sounds/Attenuation/Fort_Weapon_Rocket_Distant_Explosion_Athena_Att}
 }
 {
  id=Attenuation_4
  children [Modulator_3]
  properties {attenuationSettings=&Fort_Weapon_Rocket_Medium_Explosion_Athena_Att}
 }
 {id=Mixer_0, children [SoundClass_0,SoundClass_1,SoundClass_2], properties {inputVolume [1,0.75,1]}}
 {id=Modulator_0, children [WavePlayer_3]}
 {id=Modulator_2, children [WavePlayer_4]}
 {id=Modulator_3, children [Random_1]}
 {id=Random_1, children [WavePlayer_0,WavePlayer_1,WavePlayer_2], properties {weights [1,1,1]}}
 {id=SoundClass_0, children [Attenuation_4], properties {soundClassOverride=/Game/Sounds/FortSoundClassesAndModes/Fort_Explosives_Medium_PSM}}
 {id=SoundClass_1, children [Attenuation_0], properties {soundClassOverride=&Fort_Explosives_Medium_PSM}}
 {id=SoundClass_2, children [Attenuation_3], properties {soundClassOverride=/Game/Sounds/FortSoundClassesAndModes/Fort_ImpactSounds_Distant}}
 {id=WavePlayer_0, properties {soundWaveAssetPtr=&/Launcher_Easter_Explode_01}}
 {id=WavePlayer_1, properties {soundWaveAssetPtr=&/Launcher_Easter_Explode_02}}
 {id=WavePlayer_2, properties {soundWaveAssetPtr=&/Launcher_Easter_Explode_03}}
 {id=WavePlayer_3, properties {soundWaveAssetPtr=&/Launcher_Easter_Explode_Distant_01}}
 {id=WavePlayer_4, properties {soundWaveAssetPtr=&/Launcher_Easter_Explode_Medium_01}}
]
properties {maxDistance=15000, soundClassObject=&Fort_WeaponRanged_Fire_Distant, volumeMultiplier=0.2}
