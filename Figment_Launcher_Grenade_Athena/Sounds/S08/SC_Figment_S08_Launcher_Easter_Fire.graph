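# Sound cue for the Figment grenade launcher's Easter variant fire sound, layered by distance.
# The close layer picks one of three waves at random, modulates it and routes it through close attenuation into the third-person ranged fire sound class.
# Medium and distant single-wave layers, each with its own pitch or volume modulation and mono attenuation, are mixed into the distant fire class.
# A top mixer blends both branches, weighting the close layer higher.
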SoundCue /Figment_Launcher_Grenade_Athena/Sounds/S08/SC_Figment_S08_Launcher_Easter_Fire
root=Mixer_1
nodes [
 {id=Attenuation_0, children [Modulator_0], properties {attenuationSettings=&Fort_Weapon_Projectile_Close_Athena_Att}}
 {
  id=Attenuation_1
  children [Modulator_1]
  properties {attenuationSettings=&Fort_Weapon_Projectile_Medium_Mono_Athena_Att}
 }
 {
  id=Attenuation_2
  children [Modulator_2]
  properties {attenuationSettings=&Fort_Weapon_Projectile_Distant_Mono_Athena_Att}
 }
 {id=Mixer_0, children [Attenuation_1, Attenuation_2], properties {inputVolume [1,0.7]}}
 {id=Mixer_1, children [SoundClass_0,SoundClass_1], properties {inputVolume [1.1,0.9]}}
 {id=Modulator_0, children [Random_2]}
 {id=Modulator_1, children [WavePlayer_4], properties {pitchMax=1.1, volumeMax=0.9, volumeMin=0.8}}
 {id=Modulator_2, children [WavePlayer_3], properties {pitchMax=0.9, pitchMin=0.8}}
 {id=Random_2, children [WavePlayer_0,WavePlayer_1,WavePlayer_2], properties {weights [1,1,1]}}
 {
  id=SoundClass_0
  children [Attenuation_0]
  properties {soundClassOverride=&Fort_WeaponRanged_Fire_3P_PSM}
 }
 {id=SoundClass_1, children [Mixer_0], properties {soundClassOverride=&Fort_WeaponRanged_Fire_Distant}}
 {id=WavePlayer_0, properties {soundWaveAssetPtr=&/Launcher_Easter_Fire_01}}
 {id=WavePlayer_1, properties {soundWaveAssetPtr=&/Launcher_Easter_Fire_02}}
 {id=WavePlayer_2, properties {soundWaveAssetPtr=&/Launcher_Easter_Fire_03}}
 {id=WavePlayer_3, properties {soundWaveAssetPtr=&/Launcher_Easter_Fire_Distant_01}}
 {id=WavePlayer_4, properties {soundWaveAssetPtr=&/Launcher_Easter_Fire_Medium_01}}
]
properties {maxDistance=27500, volumeMultiplier=1}
