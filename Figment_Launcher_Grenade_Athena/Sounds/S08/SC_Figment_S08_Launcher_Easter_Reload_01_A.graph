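# Sound cue for the reload sound of the Figment Season 8 Easter grenade launcher.
# A mixer layers the stock grenade launcher shell-insert wave with an Easter-specific reload wave at half volume, which a fixed delay holds back briefly.
# It plays as weapon foley using the shared weapon reload attenuation.
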SoundCue /Figment_Launcher_Grenade_Athena/Sounds/S08/SC_Figment_S08_Launcher_Easter_Reload_01_A
root=Mixer_0
nodes [
 {id=Delay_0, children [WavePlayer_2], properties {delayMax=0.8, delayMin=0.8}}
 {id=Mixer_0, children [WavePlayer_1,Delay_0], properties {inputVolume [1,0.5]}}
 {id=WavePlayer_1, properties {soundWaveAssetPtr=/Game/Sounds/Fort_Weapons/Launchers/grenadelauncher_reload_insert_shell}}
 {id=WavePlayer_2, properties {soundWaveAssetPtr=&/Launcher_Easter_Reload_01_C}}
]
properties {
 attenuationSettings=&Fort_Weapon_Reload_Att
 maxDistance=2250
 soundClassObject=&Fort_WeaponFoley
}
