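# Sound cue for one variation of the Figment Season 8 Easter grenade launcher reload sound, going by its name and paths.
# A single wave player feeds the Launcher_Easter_Reload_01_B wave straight to the output, with no randomization or modulation.
# It uses the shared Fortnite weapon reload attenuation and weapon foley sound class, with a reduced volume multiplier.
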SoundCue /Figment_Launcher_Grenade_Athena/Sounds/S08/SC_Figment_S08_Launcher_Easter_Reload_01_B
root=WavePlayer_0
nodes [{id=WavePlayer_0, properties {soundWaveAssetPtr=&/Launcher_Easter_Reload_01_B}}]
properties {
 attenuationSettings=&Fort_Weapon_Reload_Att
 maxDistance=2250
 soundClassObject=&Fort_WeaponFoley
 volumeMultiplier=0.15
}
