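# Sound cue for a weapon drop in the Figment grenade launcher content, though the cue itself is named for a shotgun.
# A single wave player plays a shared Fortnite loot drop sound at reduced volume, with no randomization or modulation.
# It uses the weapon drop attenuation asset and the weapon foley sound class.
# A concurrency override allows one instance at a time, stopping the farthest first.
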SoundCue /Figment_Launcher_Grenade_Athena/Sounds/S08/SC_Figment_Shotgun_Drop_Cue
root=WavePlayer_0
nodes [{id=WavePlayer_0, properties {soundWaveAssetPtr=/Game/Sounds/Fort_GamePlay_Sounds/Loot/fort_weapon_drop_02_mix}}]
properties {
 attenuationSettings=&Fort_Weapon_Drop_Att
 bOverrideConcurrency=true
 maxDistance=3000
 soundClassObject=&Fort_WeaponFoley
 volumeMultiplier=0.5
 attenuationOverrides {falloffDistance=5000, attenuationShapeExtents {x=1000}}
 concurrencyOverrides {maxCount=1, resolutionRule=StopFarthestThenPreventNew}
}
