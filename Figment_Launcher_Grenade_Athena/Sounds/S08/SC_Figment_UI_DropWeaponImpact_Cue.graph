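# Sound cue for the impact sound of a dropped weapon, in the Figment grenade launcher content.
# A single wave player feeds UI_DropWeaponImpact01 straight to the output, with no randomization or modulation.
# It plays at reduced volume with weapon pickup attenuation.
# Its sound class is the distant ranged weapon fire class, despite the UI drop-impact naming.
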SoundCue /Figment_Launcher_Grenade_Athena/Sounds/S08/SC_Figment_UI_DropWeaponImpact_Cue
root=WavePlayer_1
nodes [{id=WavePlayer_1, properties {soundWaveAssetPtr=&/UI_DropWeaponImpact01}}]
properties {
 attenuationSettings=&Fort_Weapon_Pickup_Att
 maxDistance=3500
 soundClassObject=&Fort_WeaponRanged_Fire_Distant
 volumeMultiplier=0.4
}
